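# Physics asset for the Drof vehicle body mesh in the vehicle cosmetics content.
# It defines a single rigid body on the root bone, using one box shape that is offset forward and upward from the bone origin.
# There are no constraints, so the whole vehicle body collides as one simple box.
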PhysicsAsset /VehicleCosmetics/Bodies/Drof/Meshes/Body_Drof_PhysicsDM
bodies {root [{shapeType=Box, extentX=444, extentY=216, extentZ=132, shapeName=root_box, center=<31 0 101>}]}
constraints []
